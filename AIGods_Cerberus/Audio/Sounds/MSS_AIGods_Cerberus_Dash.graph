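# One-shot mono MetaSound source for the dash sound of the Cerberus in AIGods_Cerberus.
# On play, a random array getter with recent-repeat avoidance picks one of five dash waves and feeds a stereo wave player, pitched up by a random float.
# Only the player's left channel reaches the output, and playback end fires OnFinished.
# It uses an NPC boss attenuation asset and a melee weapon sound class.
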MetaSoundSource /AIGods_Cerberus/Audio/Sounds/MSS_AIGods_Cerberus_Dash
inputs [
 In Array {type=WaveAsset:Array}
 UE.Source.OnPlay {type=Trigger}
]
interfaces [UE.OutputFormat.Mono:1.0, UE.Source.OneShot:1.0, UE.Source:1.1]
outputs [
 "UE.OutputFormat.Mono.Audio:0" {type=Audio}
 UE.Source.OneShot.OnFinished {type=Trigger}
]
pages [Default {
  input_defaults {
   UE.Source.OnPlay=null
   In Array [
    &/AiGods_Cerberus_Dash_05
    &/AiGods_Cerberus_Dash_04
    &/AiGods_Cerberus_Dash_03
    &/AiGods_Cerberus_Dash_02
    &/AiGods_Cerberus_Dash_01.AIGods_Cerberus_Dash_01
   ]
  }
  nodes [
   ArrayRandomGet_0 = Array.Random Get:WaveAsset:Array(In Array=^In Array, Next=^UE.Source.OnPlay) {also {defaults {No Repeats=3}}}
   WavePlayer_0 = UE.Wave Player:Stereo(Play=@ArrayRandomGet_0.On Next, Wave Asset=@ArrayRandomGet_0.Value, Pitch Shift=UE.RandomFloat() {also {defaults {Max=2, Min=1}}}.Value)
  ]
  links [{dst="@UE.Source.OneShot.OnFinished", src=WavePlayer_0.On Finished}, {dst="@UE.OutputFormat.Mono.Audio:0", src=WavePlayer_0.Out Left}]
 }]
properties {
 assetImportData=&self:AssetImportData
 attenuationSettings=/NPCLibrary/Audio/TinyBalloon/NPC_Evil_Boss_Att
 maxDistance=3800
 sampleRate=48000
 soundClassObject=&Fort_WeaponMelee_PSM
 volume=2
}
