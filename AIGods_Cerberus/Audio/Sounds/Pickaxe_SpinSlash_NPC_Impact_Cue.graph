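# Sound cue for the impact of an NPC pickaxe spin-slash attack in the Cerberus AI content.
# A short fixed delay feeds an equal-weight random pick among six Hightower Wasabi pickaxe impact waves.
# The cue plays slightly pitched down, routes to the impacts sound class, and reuses the player pickaxe impact attenuation.
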SoundCue /AIGods_Cerberus/Audio/Sounds/Pickaxe_SpinSlash_NPC_Impact_Cue
root=Delay_0
nodes [
 {id=Delay_0, children [Random_0], properties {delayMax=0.1, delayMin=0.1}}
 {id=Random_0, children [WavePlayer_0,WavePlayer_1,WavePlayer_2,WavePlayer_3,WavePlayer_4,WavePlayer_5], properties {weights [1,1,1,1,1,1]}}
 {id=WavePlayer_0, properties {soundWaveAssetPtr=/Game/Athena/Sounds/Weapons/PickAxes/Hightower_Wasabi/Pickaxe_HightowerWasabi_Impact_01}}
 {id=WavePlayer_1, properties {soundWaveAssetPtr=/Game/Athena/Sounds/Weapons/PickAxes/Hightower_Wasabi/Pickaxe_HightowerWasabi_Impact_02}}
 {id=WavePlayer_2, properties {soundWaveAssetPtr=/Game/Athena/Sounds/Weapons/PickAxes/Hightower_Wasabi/Pickaxe_HightowerWasabi_Impact_03}}
 {id=WavePlayer_3, properties {soundWaveAssetPtr=/Game/Athena/Sounds/Weapons/PickAxes/Hightower_Wasabi/Pickaxe_HightowerWasabi_Impact_04}}
 {id=WavePlayer_4, properties {soundWaveAssetPtr=/Game/Athena/Sounds/Weapons/PickAxes/Hightower_Wasabi/Pickaxe_HightowerWasabi_Impact_05}}
 {id=WavePlayer_5, properties {soundWaveAssetPtr=/Game/Athena/Sounds/Weapons/PickAxes/Hightower_Wasabi/Pickaxe_HightowerWasabi_Impact_06}}
]
properties {
 attenuationSettings=/Game/Sounds/Attenuation/Fort_Weapon_Pickaxe_Player_Impact_Att
 maxDistance=3512
 pitchMultiplier=0.9
 soundClassObject=&Fort_Impacts
}
