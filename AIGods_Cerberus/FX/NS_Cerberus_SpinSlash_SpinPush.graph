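# Niagara effect for Cerberus's SpinSlash spin push: a brief, mostly green swirl of meshes with a distortion pulse.
# Sharp_Streaks spawns spinning, growing meshes in two spawn groups that set renderer visibility tags, with curve-driven color and material parameters.
# Distortion bursts one expanding, camera-offset sprite.
# Center_Warp adds delayed rotating meshes that widen with particle and loop age, suppressed entirely when User.IsSmallJump is true.
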NiagaraSystem /AIGods_Cerberus/FX/NS_Cerberus_SpinSlash_SpinPush
dependencies {
 usedDataInterfaces [&NiagaraDataInterfaceCurve, &NiagaraDataInterfaceColorCurve, &NiagaraDataInterfaceVectorCurve]
 usedDynamicInputs [&SimulationPosition]
 usedModules [
  &self:SystemSpawnScript.SystemScriptSource.SystemScriptGraph.NiagaraNodeAssignment_0.SetVariables_#1
  &SystemState
 ]
}
emitters {
 Center_Warp {
  emitterUpdateScript [
   &EmitterState(
    Life Cycle Mode=::NewEnumerator1
    Loop Behavior=::NewEnumerator1
    Loop Delay=0.1
    Loop Duration=&Lerp_Float(A=0.1, B=0.3, Alpha=&MakeCustomFloatFromBool(Bool=$User.IsSmallJump, True Float=0, False Float=1))
    UseLoopDelay=true
   )
   &SpawnRate(
    Spawn Probability=&MakeCustomFloatFromBool(Bool=$User.IsSmallJump, True Float=0, False Float=1)
    SpawnRate=15
    Use Spawn Probability=true
   )
   &SpawnBurst_Instantaneous(
    Spawn Count=&RandomRangeInt(Minimum=3, Maximum=3)
    Spawn Probability=&MakeCustomFloatFromBool(Bool=$User.IsSmallJump, True Float=0, False Float=1)
    Use Spawn Probability=true
    UseLoopCountLimit=false
   )
  ]
  particleSpawnScript [
   &InitializeParticle(
    Color Maximum=<rgba 0.0298 0.3229 0.3068 0.5>
    Color Minimum=<rgba 0.1073 1.7 0.7709 0.5>
    Color Mode=::NewEnumerator2
    Lifetime Max=0.4
    Lifetime Min=0.25
    Lifetime Mode=::NewEnumerator1
    Mesh Scale Mode=::NewEnumerator3
    Mesh Uniform Scale=&Multiply_Float(A=&RandomRangeFloat(Minimum=2.4, Maximum=3), B=2.5)
    Position Offset=<0 0 50>
    UsePositionOffset=false
    Write Material Random=true
   )
   &ShapeLocation(Box Size=<10 10 150>, MaskSpawnGroup=false, Override Seed=false, Shape Primitive=::NewEnumerator2, Surface Only=false, UseOwnerScale=true)
   &InitialMeshOrientation(
    Mesh Orientation Mode=::NewEnumerator1
    Rotation=&RandomRangeVector(Minimum=<-0.05 0 -1>, Maximum=<0.05 0 1>)
    Use Rotation Vector=true
   )
   &SystemLocation(MaskSpawnGroup=false, Offset=<0 0 150>)
  ]
  particleUpdateScript [
   &ParticleState()
   &DynamicMaterialParameters(
    Index 0 Param 1=&FloatFromCurve(FloatCurve=DataInterface() {propertyValues {Curve {keys [
         {arriveTangent=-0.095, interpMode=RCIM_Cubic, leaveTangent=-0.095, tangentMode=RCTM_User, time=0, value=0}
         {arriveTangent=2.534, interpMode=RCIM_Cubic, leaveTangent=2.534, tangentMode=RCTM_User, time=1, value=1}
        ]}}})
    Index 0 Param 2=&MakeCustomFloatFromBool(Bool=true, True Float=1, False Float=0)
    Index 0 Param 3=1
    Index 0 Param 4=&MakeCustomFloatFromBool(Bool=true, True Float=1, False Float=0)
    Module.Param0WriteEnabled=true
    Module.Param1WriteEnabled=false
    Module.Param2WriteEnabled=true
    Module.Param3WriteEnabled=false
   )
   /SetVariables(
    Particles.XY:NiagaraFloat=&FloatFromCurve(
     FloatCurve=DataInterface() {propertyValues {Curve {keys [
         {arriveTangent=0.2017, interpMode=RCIM_Cubic, leaveTangent=0.2017, tangentMode=RCTM_User, time=0, value=0.2}
         {arriveTangent=0.4623, interpMode=RCIM_Cubic, leaveTangent=0.4623, tangentMode=RCTM_User, time=1, value=1}
        ]}}}
     Scale Curve=&Lerp_Float(
      A=0.5
      B=1.2
      Alpha=&FloatFromCurve(
       FloatCurve=DataInterface() {propertyValues {Curve {keys [
           {interpMode=RCIM_Cubic, tangentMode=RCTM_User, time=0, value=0}
           {arriveTangent=0.7348, interpMode=RCIM_Cubic, leaveTangent=0.7348, tangentMode=RCTM_User, time=0.7, value=0.25}
           {arriveTangent=1.0206, interpMode=RCIM_Cubic, leaveTangent=1.0206, tangentMode=RCTM_User, time=0.9, value=0.5}
          ]}}}
       CurveIndex=$Emitter.NormalizedLoopAge
      )
     )
    )
    Particles.Z:NiagaraFloat=&FloatFromCurve(
     FloatCurve=DataInterface() {propertyValues {Curve {keys [
         {arriveTangent=3.7114, interpMode=RCIM_Cubic, leaveTangent=3.7114, tangentMode=RCTM_User, time=0, value=0}
         {arriveTangent=0.2081, interpMode=RCIM_Cubic, leaveTangent=0.2081, tangentMode=RCTM_User, time=0.4, value=0.525}
         {arriveTangent=-1.1305, interpMode=RCIM_Cubic, leaveTangent=-1.1305, tangentMode=RCTM_User, time=1, value=0}
        ]}}}
     Scale Curve=2
    )
   ) {isSetParameters=true}
   &ScaleMeshSize(Scale Factor=&MakeVector(X=$Particles.XY, Y=$Particles.XY, Z=$Particles.Z))
   &UpdateMeshOrientation(Rotation Rate=&RandomRangeFloat(Minimum=-2.5, Maximum=-3.5), Rotation Vector=<0 0 1>)
   &ScaleColor(
    Scale Alpha=&FloatFromCurve(
     FloatCurve=DataInterface() {propertyValues {Curve {keys [
         {interpMode=RCIM_Cubic, leaveTangentWeight=0.1333, tangentMode=RCTM_User, time=0.1, value=0}
         {arriveTangentWeight=0.342, interpMode=RCIM_Cubic, leaveTangentWeight=0.4274, tangentMode=RCTM_User, time=0.3, value=1}
         {arriveTangentWeight=0.2397, interpMode=RCIM_Cubic, tangentMode=RCTM_User, time=1, value=0}
        ]}}}
     Scale Curve=0.25
    )
    Scale RGB=&VectorFromFloat(Value=&FloatFromCurve(FloatCurve=DataInterface() {propertyValues {Curve {keys [{arriveTangent=-1.9639, time=0, value=1}, {arriveTangent=0.35, leaveTangent=0.35, time=1, value=1}]}}}))
    ScaleA=true
    ScaleRGB=true
   )
   &CameraOffset(Camera Offset Amount=15)
  ]
  rendererClasses [&NiagaraMeshRendererProperties]
  renderers [{rendererClass=&NiagaraMeshRendererProperties, rendererIndex=0}]
 }
 Distortion {
  emitterUpdateScript [
   &EmitterState()
   &SpawnBurst_Instantaneous(Spawn Count=1, Use Spawn Probability=false)
  ]
  particleSpawnScript [
   &InitializeParticle(
    Lifetime=0.5
    Mass Mode=::NewEnumerator1
    Sprite Size=<880 880>
    Sprite Size Mode=::NewEnumerator4
    UsePositionOffset=false
    Write Material Random=true
   ) {module=InitializeParticle001}
   &SystemLocation(MaskSpawnGroup=false, Offset=<0 0 175>)
  ]
  particleUpdateScript [
   &ParticleState()
   &SolveForcesAndVelocity(Clamp Velocity=false, Limit Acceleration=false)
   &ScaleColor(
    Scale RGB=&MakeVectorFromLinearColorRGB(Color=&ColorFromCurve(ColorCurve=DataInterface() {propertyValues {
        AlphaCurve {keys [<at 0 0.2>, <at 0.0501 1>, {arriveTangent=-0.2, interpMode=RCIM_Cubic, leaveTangent=-0.2, time=0.3304, value=1}, <at 1 0 RCIM_Cubic>]}
        BlueCurve {keys [<at 0 0.9392 RCIM_Cubic>,<at 1 1 RCIM_Cubic>]}
        GreenCurve {keys [<at 0 0.8734 RCIM_Cubic>,<at 1 0.5732 RCIM_Cubic>]}
        RedCurve {keys [<at 0 0.8528 RCIM_Cubic>,<at 1 0.1399 RCIM_Cubic>]}
       }}))
    ScaleA=false
    ScaleRGB=true
   )
   &ScaleSpriteSize(Scale Factor=&Vector2DFromFloat(Value=&FloatFromCurve(
      FloatCurve=DataInterface() {propertyValues {Curve {keys [{arriveTangent=2.2552, interpMode=RCIM_Cubic, leaveTangent=2.2552, tangentMode=RCTM_User, time=0, value=0}, <at 0.5 1>]}}}
      Scale Curve=0.65
     )))
   &CameraOffset(Camera Offset Amount=40)
   &DynamicMaterialParameters(
    Index 0 Param 1=&FloatFromCurve(FloatCurve=DataInterface() {propertyValues {Curve {keys [
         {arriveTangent=-0.6727, interpMode=RCIM_Cubic, leaveTangent=-0.6727, tangentMode=RCTM_User, time=0, value=1}
         <at 0.1 0.965 RCIM_Cubic>
         {arriveTangent=0.064, interpMode=RCIM_Cubic, leaveTangent=0.064, tangentMode=RCTM_User, time=1, value=1}
        ]}}})
    Index 0 Param 2=1
    Index 0 Param 3=1
    Index 0 Param 4=1
    Module.Param0WriteEnabled=true
    Module.Param1WriteEnabled=true
    Module.Param2WriteEnabled=true
    Module.Param3WriteEnabled=true
   )
  ]
  rendererClasses [&NiagaraSpriteRendererProperties]
  renderers [{rendererClass=&NiagaraSpriteRendererProperties}]
 }
 Sharp_Streaks {
  emitterUpdateScript [
   &EmitterState()
   &SpawnBurst_Instantaneous(Spawn Count=1, Use Spawn Probability=false)
   &SpawnRate(SpawnRate=12, Use Spawn Probability=false)
   &SpawnRate(SpawnGroup=1, SpawnRate=15, Use Spawn Probability=false) {module=SpawnRate001}
  ]
  particleSpawnScript [
   /SetVariables(Particles.SpawnGroup=$Emitter.SpawnGroup) {isSetParameters=true}
   /SetVariables(Particles.VisibilityTag=$Particles.SpawnGroup) {isSetParameters=true}
   /SetVariables(
    Particles.Mesh_CameraOffset_Streaks:NiagaraFloat=&Lerp_Float(A=60, B=120, Alpha=$Particles.NormalizedAge)
    Particles.Mesh_Color_Streaks:LinearColor=<rgba 1.5 1.5 1.5 0.5>
    Particles.Mesh_Size_Streaks:Vector3f=&Multiply_VectorByFloat(Vector=<3 3 3.5>, Float=0.75)
   ) {isSetParameters=true}
   &InitializeParticle(
    Color=<rgba 0 1 0.4167 1>
    Lifetime=&UniformRangedFloat(Minimum=0.35, Maximum=0.4)
    Mass Mode=::NewEnumerator1
    Mesh Scale=&VectorFromFloat(Value=&RandomRangeFloat(Minimum=4, Maximum=7))
    Mesh Scale Mode=::NewEnumerator4
    Sprite Size Mode=::NewEnumerator4
    UsePositionOffset=false
    Write Material Random=true
   ) {module=InitializeParticle001}
   &SystemLocation(MaskSpawnGroup=false, Offset=<0 0 175>)
   &InitialMeshOrientation(
    Mesh Orientation Mode=::NewEnumerator2
    Orientation Vector=<0 0 1>
    Rotation=&MakeVector(
     X=&UniformRangedFloat(Minimum=0, Maximum=1)
     Y=&UniformRangedFloat(Minimum=-0.05, Maximum=0.05)
     Z=&UniformRangedFloat(Minimum=0, Maximum=0)
    )
    Use Rotation Vector=true
   ) {module=InitialMeshOrientation001}
  ]
  particleUpdateScript [
   &ParticleState()
   &SolveForcesAndVelocity(Clamp Velocity=false, Limit Acceleration=false)
   &ScaleColor(
    Scale Alpha=&FloatFromCurve(FloatCurve=DataInterface() {propertyValues {Curve {keys [
         {arriveTangent=6.3626, interpMode=RCIM_Cubic, leaveTangent=6.3626, tangentMode=RCTM_User, time=0, value=0}
         <at 0.1573 0.9779 RCIM_Cubic>
         {interpMode=RCIM_Cubic, leaveTangent=-0.6839, time=1, value=0}
        ]}}})
    ScaleA=true
    ScaleRGB=false
   )
   /SetVariables(Particles.Renderer2_Color:LinearColor=&ColorFromCurve(ColorCurve=DataInterface() {propertyValues {
       AlphaCurve {keys [<at 0.0023 0>,<at 0.1305 2 RCIM_Cubic>,<at 1 0 RCIM_Cubic>]}
       BlueCurve {keys [<at 0 0.2775 RCIM_Cubic>,<at 0.4114 0.2665>]}
       GreenCurve {keys [<at 0 2 RCIM_Cubic>,<at 0.4114 0.5729>]}
       RedCurve {keys [<at 0 0.2529 RCIM_Cubic>,<at 0.4114 0.0835>]}
      }})) {isSetParameters=true}
   /SetVariables(
    Particles.Mesh_Color_Streaks:LinearColor=&MultiplyLinearColor(
     Linear Color=$Particles.Initial.Mesh_Color_Streaks
     Scale Factor=&MakeFloat4FromVectorAndFloat(
      "Vector (RGB)"=&MakeVectorFromLinearColorRGB(Color=&ColorFromCurve(ColorCurve=DataInterface() {propertyValues {
          AlphaCurve {keys [<at 0 1 RCIM_Cubic>]}
          BlueCurve {keys [<at 0 0.4167 RCIM_Cubic>,<at 0.1702 0.2524>,<at 1 0.4866 RCIM_Cubic>]}
          GreenCurve {keys [<at 0 1 RCIM_Cubic>,<at 0.1702 0.4948>,<at 1 0.5417 RCIM_Cubic>]}
          RedCurve {keys [<at 0 0 RCIM_Cubic>,<at 0.1702 0.0541>,<at 1 0 RCIM_Cubic>]}
         }}))
      "Float (Alpha)"=&FloatFromCurve(FloatCurve=DataInterface() {propertyValues {Curve {keys [
           {arriveTangent=8.8285, interpMode=RCIM_Cubic, leaveTangent=8.8285, tangentMode=RCTM_User, time=0, value=0.1}
           {interpMode=RCIM_Cubic, tangentMode=RCTM_User, time=0.15, value=1}
           {arriveTangent=-1.1765, interpMode=RCIM_Cubic, leaveTangent=-1.1765, time=0.5, value=0.9}
           {arriveTangent=-1.9792, interpMode=RCIM_Cubic, leaveTangent=-1.9792, tangentMode=RCTM_User, time=1, value=0}
          ]}}})
     )
    )
    Particles.Mesh_DynamicParameter_Streaks:Vector4f=&MakeVector4(
     X=&FloatFromCurve(FloatCurve=DataInterface() {propertyValues {Curve {keys [<at 0.5 0 RCIM_Cubic>, {arriveTangent=4.311, interpMode=RCIM_Cubic, leaveTangent=16.4602, tangentMode=RCTM_User, time=1, value=1}]}}})
     Y=0
     Z=1
     W=&FloatFromCurve(
      FloatCurve=DataInterface() {propertyValues {Curve {keys [
          {arriveTangent=0.2805, interpMode=RCIM_Cubic, time=0, value=1}
          {arriveTangent=2.4494, interpMode=RCIM_Cubic, leaveTangent=2.4494, tangentMode=RCTM_User, time=1, value=2}
         ]}}}
      Scale Curve=1.5
     )
    )
    Particles.Mesh_Size_Streaks:Vector3f=&Multiply_Vector(
     A=$Particles.Initial.Mesh_Size_Streaks
     B=&VectorFromCurve(
      VectorCurve=DataInterface() {propertyValues {
        XCurve {keys [
          {arriveTangent=3.0115, interpMode=RCIM_Cubic, leaveTangent=3.0115, tangentMode=RCTM_User, time=0, value=0.2}
          {arriveTangent=0.5305, interpMode=RCIM_Cubic, leaveTangent=0.5305, tangentMode=RCTM_User, time=0.5, value=0.9}
          <at 1 1>
         ]}
        YCurve {keys [
          {arriveTangent=3.0115, interpMode=RCIM_Cubic, leaveTangent=3.0115, tangentMode=RCTM_User, time=0, value=0.2}
          {arriveTangent=0.21, interpMode=RCIM_Cubic, leaveTangent=0.21, tangentMode=RCTM_User, time=0.5, value=0.9}
          <at 1 1>
         ]}
        ZCurve {keys [
          {arriveTangent=2.7258, interpMode=RCIM_Cubic, time=0, value=0.3}
          {arriveTangent=2.1816, interpMode=RCIM_Cubic, leaveTangent=2.1816, tangentMode=RCTM_User, time=1, value=1}
         ]}
       }}
      Scale Curve=<1.25 1.25 1.2>
     )
    )
   ) {isSetParameters=true}
   &UpdateMeshOrientation(Rotation Rate=-2) {module=UpdateMeshOrientation001}
  ]
  rendererClasses [&NiagaraMeshRendererProperties]
  renderers [
   {rendererClass=&NiagaraMeshRendererProperties, rendererIndex=0}
   {rendererClass=&NiagaraMeshRendererProperties, rendererIndex=1}
   {rendererClass=&NiagaraMeshRendererProperties, rendererIndex=2}
   {rendererClass=&NiagaraMeshRendererProperties, rendererIndex=3}
  ]
 }
}
summary {
 systemName=NS_Cerberus_SpinSlash_SpinPush
 emitters [
  {emitterName=Sharp_Streaks, rendererClasses [&NiagaraMeshRendererProperties]}
  {emitterName=Distortion, rendererClasses [&NiagaraSpriteRendererProperties]}
  {emitterName=Center_Warp, rendererClasses [&NiagaraMeshRendererProperties]}
 ]
 userVariables [
  User.FX_Scale_Impact_Center {type=NiagaraFloat, defaultValue=1, description=""}
  User.IsSmallJump {type=NiagaraBool}
  User.RibbonIDOffset {type=NiagaraInt32}
  User.SkeletalMesh {
   type=NiagaraDataInterfaceSkeletalMesh
   defaultValue {
    struct=NiagaraExt_VariableValue_DataInterface
    value {dataInterface=&self:NiagaraDataInterfaceSkeletalMesh_0}
   }
  }
 ]
}
